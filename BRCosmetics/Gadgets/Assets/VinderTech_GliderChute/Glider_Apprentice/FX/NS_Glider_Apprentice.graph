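# Niagara system for the Apprentice glider, drawing flickering mesh lightning over the glider's skeletal mesh.
# Arc bolts ride rings rotated by sampled socket or bone rotations, while center bolts stretch between inner and outer rings via GetLightningOrientation, adding glow sprites.
# Time-smoothed system values modulate spawn rates, bolt lifetime and center offset; User.VariantPicker feeds materials and glow tint.
# Two disabled debug sprite emitters visualize those smoothed values.
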NiagaraSystem /BRCosmetics/Gadgets/Assets/VinderTech_GliderChute/Glider_Apprentice/FX/NS_Glider_Apprentice
dependencies {
 usedDataInterfaces [&NiagaraDataInterfaceCurve, &NiagaraDataInterfaceColorCurve, &NiagaraDataInterfaceArrayFloat]
 usedDynamicInputs [
  &SetBoolByFloatComparison
  &FN_FrontEndFloatModifier
  &Abs_Float
 ]
 usedModules [&SystemState, &Timeline, &InterpolateOverTime]
}
emitters {
 FN_Debug {
  bEnabled=false
  emitterUpdateScript [
   &EmitterState(Life Cycle Mode=::NewEnumerator1, Loop Behavior=::NewEnumerator1, Loop Duration Mode=::NewEnumerator1, UseLoopDelay=false)
   &SpawnBurst_Instantaneous(Spawn Count=1, Use Spawn Probability=false)
  ]
  particleSpawnScript [/Niagara/Modules/Spawn/Initialization/InitializeParticle(
    Lifetime=5
    Module.Write Color=true
    Module.Write Lifetime=true
    Module.Write Mass=true
    Module.Write Material Random=true
    Module.Write Position=false
    Module.Write Scale=false
    Module.Write SpriteRotation=false
    Module.Write SpriteSize=true
    Position=&SimulationPosition()
    Sprite Size=<100 100>
   )]
  particleUpdateScript [
   &ParticleState(Kill Particles When Lifetime Has Elapsed=false)
   &FN_Timer()
   /SetVariables(Particles.Debug:NiagaraFloat=$System.InterpolateOverTime001.InterpolateOverTime_Float.Moving Average) {isSetParameters=true, module=SetVariables_#1}
   &DynamicMaterialParameters(
    Index 0 Param 1=$Particles.Debug
    Module.Param0WriteEnabled=true
    Module.Param1WriteEnabled=true
    Module.Param2WriteEnabled=true
    Module.Param3WriteEnabled=true
   )
   &CameraOffset(Camera Offset Amount=111)
   &Color(Color=&LerpLinearColors(
     StartColor=<rgba 1 0 0 1>
     EndColor=<rgba 0 1 1 1>
     LerpFactor=&RemapRange(
      Input Value To Remap=&Multiply_Float(A=$System.Timeline.Float, B=$User.ForwardDot)
      Input Min=-1
      Input Max=1
      Output Min=0
      Output Max=1
     )
    ))
  ]
  rendererClasses [&NiagaraSpriteRendererProperties]
  renderers [{rendererClass=&NiagaraSpriteRendererProperties}]
 }
 FN_Debug001 {
  bEnabled=false
  emitterUpdateScript [
   &EmitterState(Life Cycle Mode=::NewEnumerator1, Loop Behavior=::NewEnumerator1, Loop Duration Mode=::NewEnumerator1, UseLoopDelay=false)
   &SpawnBurst_Instantaneous(Spawn Count=1, Use Spawn Probability=false)
  ]
  particleSpawnScript [&InitializeParticle(
    Lifetime=5
    Module.Write Color=true
    Module.Write Lifetime=true
    Module.Write Mass=true
    Module.Write Material Random=true
    Module.Write Position=true
    Module.Write Scale=false
    Module.Write SpriteRotation=false
    Module.Write SpriteSize=true
    Position=&AddVectorToPosition(Position=&SimulationPosition(), Vector=<0 0 100>)
    Sprite Size=<100 100>
   )]
  particleUpdateScript [
   &ParticleState(Kill Particles When Lifetime Has Elapsed=false)
   &FN_Timer()
   /SetVariables(Particles.Debug:NiagaraFloat=$System.InterpolateOverTime.InterpolateOverTime_Float.Moving Average) {isSetParameters=true, module=SetVariables_#1}
   &DynamicMaterialParameters(
    Index 0 Param 1=$Particles.Debug
    Module.Param0WriteEnabled=true
    Module.Param1WriteEnabled=true
    Module.Param2WriteEnabled=true
    Module.Param3WriteEnabled=true
   )
   &CameraOffset(Camera Offset Amount=111)
   &Color(Color=&LerpLinearColors(
     StartColor=<rgba 1 0 0 1>
     EndColor=<rgba 0 1 1 1>
     LerpFactor=&RemapRange(
      Input Value To Remap=$System.InterpolateOverTime.InterpolateOverTime_Float.Moving Average
      Input Min=-1
      Input Max=1
      Output Min=0
      Output Max=1
     )
    ))
  ]
  rendererClasses [&NiagaraSpriteRendererProperties]
  renderers [{rendererClass=&NiagaraSpriteRendererProperties}]
 }
 NE_ArcsOutLightnings {
  emitterUpdateScript [
   &EmitterState(
    Inactive Response=::NewEnumerator1
    Life Cycle Mode=::NewEnumerator1
    Loop Behavior=::NewEnumerator1
    Loop Duration Mode=::NewEnumerator1
    UseLoopDelay=false
   )
   &SpawnRate(
    SpawnRate=&Lerp_Float(A=5, B=7, Alpha=&ClampFloat(Float=$System.InterpolateOverTime001.InterpolateOverTime_Float.Moving Average))
    Use Spawn Probability=false
   )
  ]
  particleSpawnScript [
   /Niagara/Modules/Spawn/Initialization/V2/InitializeParticle(
    Color=<rgba 0.4072 1 1 1>
    Lifetime Max=0.6
    Lifetime Min=0.4
    Lifetime Mode=::NewEnumerator1
    Mesh Scale=&RandomRangeVector(Minimum=<15 10 50>, Maximum=<35 10 75>)
    Mesh Scale Mode=::NewEnumerator4
    Sprite Size Mode=::NewEnumerator4
    UsePositionOffset=false
    Write Material Random=true
   )
   &FN_SetSpawnGroup(Number of Spawn Groups=2)
   &InitialMeshOrientation(
    Mesh Orientation Mode=::NewEnumerator7
    Rotation=&RandomRangeVector(Minimum=<0 0 -0.075>, Maximum=<0 0 0.075>)
    Use Rotation Vector=true
   )
   /SetVariables(Particles.RingDistribution:NiagaraFloat=&MaskFloatBySpawnGroup(
     Masked Float=&RandomRangeFloat(Minimum=-0.25, Maximum=0.25)
     Passthrough Float=&MaskFloatBySpawnGroup(
      Masked Float=&RandomRangeFloat(Minimum=0.25, Maximum=0.75)
      Passthrough Float=&MaskFloatBySpawnGroup(
       Masked Float=&RandomRangeFloat(Minimum=-0.25, Maximum=0.25)
       Passthrough Float=&RandomRangeFloat(Minimum=0.25, Maximum=0.75)
       Spawn Group Mask=2
      )
      Spawn Group Mask=1
     )
    )) {isSetParameters=true}
   /SetVariables(Particles.RingDistribution:NiagaraFloat=&FloatFromCurve(
     FloatCurve=DataInterface() {propertyValues {Curve {keys [<at 0 1>,<at 1 0>]}}}
     CurveIndex=$System.NormalizedLoopAge
     Scale Curve=&Lerp_Float(A=10, B=15, Alpha=&ClampFloat(Float=$System.InterpolateOverTime001.InterpolateOverTime_Float.Moving Average))
    )) {isSetParameters=true, module=SetVariables_#2}
   /SetVariables(Particles.CustomMeshIndex:NiagaraInt32=&RandomRangeInt(Minimum=0, Maximum=1)) {isSetParameters=true, module=SetVariables_#3}
  ]
  particleUpdateScript [
   &ParticleState()
   &SampleSkeletalMesh(
    AdditionalYawPitchRoll=false
    Bone / Socket Index:NiagaraInt32=&SpawnGroup()
    Bone Sampling Mode:ENiagaraBoneSamplingMode=::NewEnumerator4
    Skeletal Mesh:NiagaraDataInterfaceSkeletalMesh=$User.SkeletalMesh
   ) {module=SampleSkeletalMesh002}
   &ShapeLocation(
    Additional Quaternion Rotation=false
    MaskSpawnGroup=false
    Override Seed=false
    Ring / Disc Distribution Mode:ENiagaraSphereDistributionMode=::NewEnumerator1
    Ring Radius=95
    Rotation Mode=::NewEnumerator3
    Rotation Quaternion=&MultiplyQuaternion(Quaternion A=$Particles.SampleSkeletalMesh002.SampledRotation, Quaternion B=<1 0 0 1>)
    Shape Primitive=::NewEnumerator4
    U Position=$Particles.RingDistribution
    UseOwnerScale=true
   ) {module=ShapeLocation001}
   &UpdateMeshOrientation(
    Facing Direction=&NormalizeVector(
     Vector To Normalize=$Particles.ShapeLocation001.ShapeVector
     Safe Normalize=true
     Safe Fallback Threshold=0
     Safe Fallback Vector=<0 0 0.001>
    )
    Facing Interpolation=false
    Orientation Method=::NewEnumerator1
    Override Quat=false
    Side Direction:Vector3f=<0 0 1>
    Stabilization Mode:ENiagara_OrientVectorCount=::NewEnumerator2
    TransformFacing=false
    TransformSide=false
   ) {module=UpdateMeshOrientation001}
   &DynamicMaterialParameters(
    Index 0 Param 1=&RandomRangeFloat(Minimum=0.8, Maximum=1)
    Index 0 Param 3=$User.VariantPicker
    Index 0 Param 4=&FloatFromCurve(FloatCurve=DataInterface() {propertyValues {Curve {keys [<at 0 0>,<at 1 1>]}}})
    Index 1 Param 1=&Multiply_Float(A=$Particles.NormalizedAge, B=&RandomRangeFloat(Minimum=0.1, Maximum=0.2))
    Index 1 Param 2=&Multiply_Float(A=$Particles.NormalizedAge, B=&RandomRangeFloat(Minimum=0, Maximum=0))
    Module.Param0WriteEnabled=true
    Module.Param0WriteEnabled1=true
    Module.Param1WriteEnabled=true
    Module.Param1WriteEnabled1=true
    Module.Param2WriteEnabled=true
    Module.Param2WriteEnabled1=true
    Module.Param3WriteEnabled=true
    Module.Param3WriteEnabled1=true
    Write Parameter Index 1=true
   ) {module=DynamicMaterialParameters001}
   /SetVariables(Particles.GlowDynamicParam:Vector4f=&MakeVector4(
     X=&Multiply_Float(A=&MakeFloatFromVector4(Vector 4=$Particles.DynamicMaterialParameter, Channel=::NewEnumerator0), B=3)
     Y=1
     Z=-1
     W=&MakeFloatFromVector4(Vector 4=$Particles.DynamicMaterialParameter, Channel=::NewEnumerator3)
    )) {isSetParameters=true, module=SetVariables_#4}
   /SetVariables(Particles.GlowColor=&ColorFromCurve(ColorCurve=DataInterface() {propertyValues {
       AlphaCurve {keys [<at 0 2 RCIM_Cubic>]}
       BlueCurve {keys [<at 0 2 RCIM_Cubic>]}
       GreenCurve {keys [<at 0 0.5909 RCIM_Cubic>]}
       RedCurve {keys [<at 0 0 RCIM_Cubic>]}
      }})) {isSetParameters=true, module=SetVariables_#5}
   /SetVariables(Particles.OutlineDynamicParam:Vector4f=&MakeVector4(
     X=&Multiply_Float(A=&MakeFloatFromVector4(Vector 4=$Particles.DynamicMaterialParameter, Channel=::NewEnumerator0), B=1.1)
     Y=0
     Z=-2
     W=&MakeFloatFromVector4(Vector 4=$Particles.DynamicMaterialParameter, Channel=::NewEnumerator3)
    )) {isSetParameters=true, module=SetVariables_#6}
   /SetVariables(Particles.OutlineColor:LinearColor=&ColorFromCurve(ColorCurve=DataInterface() {propertyValues {
       AlphaCurve {keys [<at 0 2 RCIM_Cubic>]}
       BlueCurve {keys [<at 0 2 RCIM_Cubic>]}
       GreenCurve {keys [<at 0 0.487 RCIM_Cubic>]}
       RedCurve {keys [<at 0 0 RCIM_Cubic>]}
      }})) {isSetParameters=true, module=SetVariables_#7}
  ]
  rendererClasses [&NiagaraMeshRendererProperties]
  renderers [{rendererClass=&NiagaraMeshRendererProperties, rendererIndex=0}]
 }
 NE_CenterLightnings {
  emitterUpdateScript [
   &EmitterState(
    Inactive Response=::NewEnumerator1
    Life Cycle Mode=::NewEnumerator1
    Loop Behavior=::NewEnumerator1
    Loop Duration Mode=::NewEnumerator1
    UseLoopDelay=false
   )
   &SpawnRate(
    SpawnRate=&Lerp_Float(A=8, B=9, Alpha=&ClampFloat(Float=$System.InterpolateOverTime001.InterpolateOverTime_Float.Moving Average))
    Use Spawn Probability=false
   )
  ]
  particleSpawnScript [
   &InitializeParticle(
    Color=<rgba 0.4072 1 1 1>
    Lifetime=&Lerp_Float(
     A=&RandomRangeFloat(Minimum=0.4, Maximum=0.5)
     B=&RandomRangeFloat(Minimum=0.3, Maximum=0.4)
     Alpha=&ClampFloat(Float=$System.InterpolateOverTime001.InterpolateOverTime_Float.Moving Average)
    )
    Mesh Scale=&RandomRangeVector(Minimum=<20 10 40>, Maximum=<25 10 60>)
    Mesh Scale Mode=::NewEnumerator4
    Sprite Size=<50 50>
    Sprite Size Mode=::NewEnumerator4
    UsePositionOffset=false
    Write Material Random=true
   )
   &FN_SetSpawnGroup(Number of Spawn Groups=2)
   &InitialMeshOrientation(
    Mesh Orientation Mode=::NewEnumerator7
    Rotation=&RandomRangeVector(Minimum=<0 0 -0.075>, Maximum=<0 0 0.075>)
    Use Rotation Vector=true
   )
   /SetVariables(Particles.RingDistribution:NiagaraFloat=&FloatFromCurve(FloatCurve=DataInterface() {propertyValues {Curve {keys [<at 0 0>,<at 1 1>]}}}, CurveIndex=$System.NormalizedLoopAge, Scale Curve=15)) {isSetParameters=true, module=SetVariables_#2}
   /SetVariables(Particles.RingDistributionOut:NiagaraFloat=&FloatFromCurve(
     FloatCurve=DataInterface() {propertyValues {Curve {keys [<at 0 -0.2>,<at 0.5 0.2>,<at 0.5 0.7>,<at 1 0.3>]}}}
     CurveIndex=&ModuloFloat(A=$Particles.RingDistribution, B=1)
    )) {isSetParameters=true}
   /SetVariables(Particles.CustomMeshIndex:NiagaraInt32=&RandomRangeInt(Minimum=0, Maximum=1)) {isSetParameters=true, module=SetVariables_#3}
  ]
  particleUpdateScript [
   &ParticleState()
   &SampleSkeletalMesh(
    AdditionalYawPitchRoll=false
    Bone / Socket Index:NiagaraInt32=&Add_Integer(A=&SpawnGroup(), B=2)
    Bone Sampling Mode:ENiagaraBoneSamplingMode=::NewEnumerator4
    Skeletal Mesh:NiagaraDataInterfaceSkeletalMesh=$User.SkeletalMesh
   ) {module=SampleSkeletalMesh001}
   &SampleSkeletalMesh(
    AdditionalYawPitchRoll=false
    Bone / Socket Index:NiagaraInt32=&SpawnGroup()
    Bone Sampling Mode:ENiagaraBoneSamplingMode=::NewEnumerator4
    Skeletal Mesh:NiagaraDataInterfaceSkeletalMesh=$User.SkeletalMesh
   ) {module=SampleSkeletalMesh002}
   &ShapeLocation(
    Additional Quaternion Rotation=false
    MaskSpawnGroup=false
    Override Seed=false
    Ring / Disc Distribution Mode:ENiagaraSphereDistributionMode=::NewEnumerator1
    Ring Radius=113
    Rotation Mode=::NewEnumerator3
    Rotation Quaternion=&MultiplyQuaternion(
     Quaternion A=$Particles.SampleSkeletalMesh002.SampledRotation
     Quaternion B=&LerpQuaternion(
      Quaternion A:Quat4f=<1 0 0 1>
      Quaternion B:Quat4f=<0 0 0 1>
      Lerp Factor:NiagaraFloat=&MaskFloatBySpawnGroup(
       Masked Float=0
       Passthrough Float=&MaskFloatBySpawnGroup(Masked Float=0, Passthrough Float=1, Spawn Group Mask=1)
      )
      Enable Spherical Linear Interpolation:NiagaraBool=false
     )
    )
    Shape Origin=$StackContext.Position
    Shape Primitive=::NewEnumerator4
    U Position=$Particles.RingDistributionOut
    UseOwnerScale=true
   ) {module=ShapeLocation001}
   &ShapeLocation(
    MaskSpawnGroup=false
    Offset Mode=::NewEnumerator0
    Override Seed=false
    Ring / Disc Distribution Mode:ENiagaraSphereDistributionMode=::NewEnumerator1
    Ring Radius=&RandomRangeFloat(Minimum=50, Maximum=50)
    Shape Primitive=::NewEnumerator4
    U Position=&Add_Float(A=$Particles.RingDistribution, B=-0.25)
    UseOwnerScale=true
   )
   &UpdateMeshOrientation(
    Facing Direction=&NormalizeVector(
     Vector To Normalize=&Subtract_Vector(A=Linked() {linkedVariable {name=Particles.Position, type=NiagaraPosition}}, B=/Niagara/DynamicInputs/Helpers/SimulationPosition())
     Safe Normalize=true
     Safe Fallback Threshold=0
     Safe Fallback Vector=<0 0 0.001>
    )
    Facing Interpolation=false
    Orientation Method=::NewEnumerator1
    Override Quat=false
    Side Direction:Vector3f=<0 0 1>
    Stabilization Mode:ENiagara_OrientVectorCount=::NewEnumerator2
    TransformFacing=false
    TransformSide=false
   ) {module=UpdateMeshOrientation001}
   &self:GetLightningOrientation(
    BaseLocation:Vector3f=&AddVectorToPosition(
     Position=&AddVectorToPosition(
      Position=&SimulationPosition()
      Vector=&TransformVector(
       Vector=&Lerp_Vector(
        A=<15 0 0>
        B=<-15 0 0>
        Alpha=&RemapRange(
         Input Value To Remap=$System.InterpolateOverTime.InterpolateOverTime_Float.Moving Average
         Input Min=-1
         Input Max=1
         Output Min=0
         Output Max=1
        )
       )
       Destination Space=::Simulation
      )
     )
     Vector=$Particles.ShapeLocation.ShapeVector
    )
    EndLocation:Vector3f=&AddVectorToPosition(Position=&SimulationPosition(), Vector=$Particles.ShapeLocation001.ShapeVector)
    OrientationTreshold=0.5
   ) {module=GetLightningOrientation001}
   &SampleSkeletalMesh(
    AdditionalYawPitchRoll=false
    Bone / Socket Index=4
    Bone Sampling Mode:ENiagaraBoneSamplingMode=::NewEnumerator4
    Skeletal Mesh:NiagaraDataInterfaceSkeletalMesh=$User.SkeletalMesh
   )
   &DynamicMaterialParameters(
    Index 0 Param 1=&RandomRangeFloat(Minimum=0.6, Maximum=0.8)
    Index 0 Param 3=$User.VariantPicker
    Index 0 Param 4=&FloatFromCurve(FloatCurve=DataInterface() {propertyValues {Curve {keys [<at 0 0>,<at 1 1>]}}})
    Index 1 Param 1=&Multiply_Float(A=$Particles.NormalizedAge, B=&RandomRangeFloat(Minimum=0.1, Maximum=0.2))
    Index 1 Param 2=&Multiply_Float(A=$Particles.NormalizedAge, B=&RandomRangeFloat(Minimum=0, Maximum=0))
    Module.Param0WriteEnabled=true
    Module.Param0WriteEnabled1=true
    Module.Param1WriteEnabled=true
    Module.Param1WriteEnabled1=true
    Module.Param2WriteEnabled=true
    Module.Param2WriteEnabled1=true
    Module.Param3WriteEnabled=true
    Module.Param3WriteEnabled1=true
    Write Parameter Index 1=true
   ) {module=DynamicMaterialParameters001}
   /SetVariables(
    Particles.BranchOrient:Quat4f=&OrientMeshToVector(Look At Direction=$Particles.LightningDirection, Reference Axis Vector=<1 0 0>)
    Particles.BranchSize:Vector3f=&Multiply_VectorByFloat(Vector=<1 1 1>, Float=$Particles.LightningLength)
   ) {isSetParameters=true}
   /SetVariables(Particles.GlowDynamicParam:Vector4f=&MakeVector4(
     X=&Multiply_Float(A=&MakeFloatFromVector4(Vector 4=$Particles.DynamicMaterialParameter, Channel=::NewEnumerator0), B=3)
     Y=1
     Z=-1
     W=&MakeFloatFromVector4(Vector 4=$Particles.DynamicMaterialParameter, Channel=::NewEnumerator3)
    )) {isSetParameters=true, module=SetVariables_#4}
   /SetVariables(Particles.GlowColor=&ColorFromCurve(ColorCurve=DataInterface() {propertyValues {
       AlphaCurve {keys [<at 0 2 RCIM_Cubic>]}
       BlueCurve {keys [<at 0 2 RCIM_Cubic>]}
       GreenCurve {keys [<at 0 0.4878 RCIM_Cubic>]}
       RedCurve {keys [<at 0 0 RCIM_Cubic>]}
      }})) {isSetParameters=true, module=SetVariables_#5}
   /SetVariables(Particles.OutlineDynamicParam:Vector4f=&MakeVector4(
     X=&Multiply_Float(A=&MakeFloatFromVector4(Vector 4=$Particles.DynamicMaterialParameter, Channel=::NewEnumerator0), B=1.1)
     Y=0
     Z=-2
     W=&MakeFloatFromVector4(Vector 4=$Particles.DynamicMaterialParameter, Channel=::NewEnumerator3)
    )) {isSetParameters=true, module=SetVariables_#6}
   /SetVariables(Particles.OutlineColor:LinearColor=&ColorFromCurve(ColorCurve=DataInterface() {propertyValues {
       AlphaCurve {keys [<at 0 2 RCIM_Cubic>]}
       BlueCurve {keys [<at 0 2 RCIM_Cubic>]}
       GreenCurve {keys [<at 0 0.4384 RCIM_Cubic>]}
       RedCurve {keys [<at 0 0 RCIM_Cubic>]}
      }})) {isSetParameters=true, module=SetVariables_#7}
   /SetVariables(Particles.PointGlowDynParam:Vector4f=<1 0.5 1 1>) {isSetParameters=true}
   /SetVariables(Particles.PointGlowColor:LinearColor=&LerpLinearColors(
     StartColor=&ColorFromCurve(ColorCurve=DataInterface() {propertyValues {
        AlphaCurve {keys [<at 0 0 RCIM_Cubic>,<at 0.3568 0>,<at 0.4519 1>,<at 0.5725 1>,<at 0.9208 0 RCIM_Cubic>]}
        BlueCurve {keys [<at 0 0.8667 RCIM_Cubic>]}
        GreenCurve {keys [<at 0 1 RCIM_Cubic>]}
        RedCurve {keys [<at 0 0 RCIM_Cubic>]}
       }})
     EndColor=&ColorFromCurve(ColorCurve=DataInterface() {propertyValues {
        AlphaCurve {keys [<at 0 0 RCIM_Cubic>,<at 0.3568 0>,<at 0.4519 1>,<at 0.5725 1>,<at 0.9208 0 RCIM_Cubic>]}
        BlueCurve {keys [<at 0 1 RCIM_Cubic>]}
        GreenCurve {keys [<at 0 0.1976 RCIM_Cubic>]}
        RedCurve {keys [<at 0 0.9941 RCIM_Cubic>]}
       }})
     LerpFactor=$User.VariantPicker
    )) {isSetParameters=true}
   /SetVariables(Particles.GlowPos:Vector3f=$Particles.LightningEndLocation) {isSetParameters=true}
   &Color(Color=&ColorFromCurve(ColorCurve=DataInterface() {propertyValues {
       AlphaCurve {keys [<at 0.1294 0 RCIM_Cubic>,<at 0.1502 1>,<at 0.8741 1>,<at 0.8998 0 RCIM_Cubic>]}
       BlueCurve {keys [<at 0 1 RCIM_Cubic>,<at 1 1 RCIM_Cubic>]}
       GreenCurve {keys [<at 0 1 RCIM_Cubic>,<at 1 1 RCIM_Cubic>]}
       RedCurve {keys [<at 0 1 RCIM_Cubic>,<at 1 1 RCIM_Cubic>]}
      }}))
  ]
  rendererClasses [&NiagaraMeshRendererProperties, &NiagaraSpriteRendererProperties]
  renderers [
   {rendererClass=&NiagaraMeshRendererProperties, rendererIndex=0}
   {rendererClass=&NiagaraMeshRendererProperties, rendererIndex=1}
   {rendererClass=&NiagaraSpriteRendererProperties, rendererIndex=2}
  ]
 }
}
summary {
 systemName=NS_Glider_Apprentice
 emitters [
  {bEnabled=false, emitterName=FN_Debug, rendererClasses [&NiagaraSpriteRendererProperties]}
  {
   emitterName=NE_CenterLightnings
   rendererClasses [&NiagaraMeshRendererProperties, &NiagaraSpriteRendererProperties]
  }
  {emitterName=NE_ArcsOutLightnings, rendererClasses [&NiagaraMeshRendererProperties]}
  {bEnabled=false, emitterName=FN_Debug001, rendererClasses [&NiagaraSpriteRendererProperties]}
 ]
 userVariables [
  User.bIsFrontEnd {type=NiagaraBool}
  User.bIsFrontEndPreview {type=NiagaraBool}
  User.bIsFullyDeployed {type=NiagaraBool}
  User.CenterScale {type=NiagaraFloat, defaultValue=0.5, description=""}
  User.ForwardDot {type=NiagaraFloat, defaultValue=0, description=""}
  User.RibbonIDOffset {type=NiagaraInt32}
  User.RightDot {type=NiagaraFloat, defaultValue=0, description=""}
  User.SkeletalMesh {
   type=NiagaraDataInterfaceSkeletalMesh
   defaultValue {
    struct=NiagaraExt_VariableValue_DataInterface
    value {dataInterface=&self:NiagaraDataInterfaceSkeletalMesh_0}
   }
  }
  User.VariantPicker {type=NiagaraFloat, defaultValue=0, description=""}
 ]
}
